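# Subsurface master material for glowing alien flora in Apollo Settlement, with Nanite and instancing support.
# Vertex-color or UV2 masks drive wind and sine sway, while time-of-day-scaled emissive pulses through color variation and scrolling spots.
# Epic quality adds a Fresnel-weighted caustics glow layer.
# FoliageActor switch replaces sway with camera-proximity push, dimming and fade sinking, otherwise adding destructible hit animation, before Curie fallback and Creative move-tool effects.
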Material /Game/Environments/Apollo/Foliage/Settlement/Materials/M_AlienFlora
graph {
 properties {
  bUseMaterialAttributes=true
  bUsedWithInstancedStaticMeshes=true
  bUsedWithNanite=true
  pixelDepthOffsetMode=PDOM_Legacy
  shadingModel=MSM_Subsurface
  usedShadingModels=Subsurface
 }
 nodes [
  n1 = VertexColor()
  n2 = Multiply(A=ScalarParameter() {defaultValue=1, group=Wind, parameterName=Motion R}, B=@n1.R)
  n3 = Multiply(A=ScalarParameter() {defaultValue=1, group=Wind, parameterName=Motion G}, B=@n1.G)
  n4 = ScalarParameter() {defaultValue=1, group=Wind, parameterName=Motion B}
  n5 = Multiply(
   A=Max(A=Max(A=@n2, B=@n3), B=Multiply(A=@n1.B, B=@n4))
   B=StaticSwitchParameter(True=NamedRerouteUsage(), False=@n1.A) {group=UV Mask, parameterName=UseUV2AnimMaskWind}
  )
  n6 = MaterialFunctionCall!(
   WindIntensity=ScalarParameter() {defaultValue=0.05, group=Wind, parameterName=Wind_Intensity}
   WindWeight=@n5
   WindSpeed=ScalarParameter() {defaultValue=0.5, group=Wind, parameterName=Wind_Speed}
   AdditionalWPO=Constant3Vector()
  ) {materialFunction=/Engine/Functions/Engine_MaterialFunctions01/WorldPositionOffset/SimpleGrassWind}
  n7 = OneMinus(None=ComponentMask(None=TextureCoordinate() {coordinateIndex=1}) {g=true})
  n8 = VertexColor()
  n9 = StaticSwitchParameter(True=NamedRerouteDeclaration(None=@n7) {name=Uv2Mask, nodeColor=<rgba 0.8281 0 1 1>}, False=@n8.A) {group=UV Mask, parameterName=UseUV2AnimMask}
  n10 = Divide(A=ActorPositionWS(), B=ScalarParameter() {defaultValue=35, group=Emissive, parameterName="Color Change Frequency (WS)", sortPriority=4})
  n11 = StaticSwitchParameter(
   True=PerInstanceRandom()
   False=MaterialFunctionCall!(f2=ComponentMask(None=@n10) {g=true, r=true}, f3=@n10, f4) {materialFunction=/Game/Packages/Fortnite_UniqueMaterials_1/VectorComponentOps}.f3 Add
  ) {group=Foliage, parameterName=FoliageActor}
  n12 = Add(A=Add(A=OneMinus(None=@n9), B=@n11), B=Time())
  n13 = ScalarParameter() {defaultValue=0.25, group=Emissive, parameterName=Flow Speed}
  n14 = TextureSampleParameter2D() {
   group=Emissive
   parameterName=Emissive
   samplerSource=SSM_Wrap_WorldGroupSettings
   samplerType=SAMPLERTYPE_Masks
   sortPriority=1
   texture=/Game/Environments/Apollo/Foliage/Settlement/Textures/T_M_SM_Settlement_Flora_04_E
  }
  n15 = ScalarParameter() {defaultValue=0.1, group=Emissive, parameterName=EmissiveToAnimation}
  n16 = Sine(None=Add(A=Multiply(A=@n12, B=@n13), B=Multiply(A=@n14.R, B=@n15)))
  n17 = Multiply(A=@n16, B=ScalarParameter() {defaultValue=1, group=Wind, parameterName=SineToMotion})
  n18 = TextureSampleParameter2D() {
   group="Base "
   parameterName=Diffuse
   samplerSource=SSM_Wrap_WorldGroupSettings
   sortPriority=0
   texture=/Game/Environments/Apollo/Foliage/Settlement/Textures/T_M_SM_Settlement_Flora_04_D
  }
  n19 = MaterialFunctionCall(Hue Shift Percentage=Divide(A=NamedRerouteUsage(), B=ScalarParameter() {defaultValue=8, group="Base ", parameterName=HUE}), Texture=@n18.RGB) {materialFunction=&HueShift}
  n20 = TextureSampleParameter2D() {
   group="Base "
   parameterName=Specular
   samplerSource=SSM_Wrap_WorldGroupSettings
   samplerType=SAMPLERTYPE_Masks
   sortPriority=1
   texture=/Game/Environments/Apollo/Foliage/Settlement/Textures/T_M_SM_Settlement_Flora_04_S
  }
  n21 = ScalarParameter() {defaultValue=10, group="Base ", parameterName=SSS_Value}
  n22 = TextureSampleParameter2D() {
   group="Base "
   parameterName=Normals
   samplerSource=SSM_Wrap_WorldGroupSettings
   samplerType=SAMPLERTYPE_Normal
   sortPriority=2
   texture=/Game/Environments/Apollo/Foliage/Settlement/Textures/T_M_SM_Settlement_Flora_04_N
  }
  n23 = SetMaterialAttributes!(
   MaterialAttributes
   Base Color=@n19.Result
   Specular=@n20.R
   Roughness=LinearInterpolate(
    A=ScalarParameter() {group="Base ", parameterName=Rough Min, sortPriority=8}
    B=ScalarParameter() {defaultValue=1, group="Base ", parameterName=Rough Max, sortPriority=8}
    Alpha=@n20.B
   )
   Opacity=Multiply(A=ScalarParameter() {defaultValue=5, group="Base ", parameterName=SSSDepth}, B=@n20.G)
   Subsurface Color=Multiply(A=@n19.Result, B=Multiply(A=@n20.G, B=@n21))
   Normal=NamedRerouteDeclaration(None=@n22.RGB) {name=Normal, nodeColor=<rgba 1 0.7266 0 1>}
  ) {attributeSetTypes [
    #1
    #2
    #3
    #4
    #5
    #6
   ]}
  n24 = VectorParameter() {group=Emissive, parameterName=ColorB, sortPriority=3, defaultValue=<rgba 0 0.5292 2 0>}
  n25 = VectorParameter() {group=Emissive, parameterName=ColorA, sortPriority=2, defaultValue=<rgba 12 0.6604 0 0>}
  n26 = NamedRerouteDeclaration(None=DotProduct(
    A=@n8
    B=TextureSample(UVs=@n11) {samplerSource=SSM_Wrap_WorldGroupSettings, texture=/Game/Environments/Apollo/Foliage/Settlement/Textures/T_RGB}.RGB
   )) {name=Random, nodeColor=<rgba 1 0 0.6562 1>}
  n27 = TextureSampleParameter2D() {
   group="Base "
   parameterName=Specular
   samplerSource=SSM_Wrap_WorldGroupSettings
   samplerType=SAMPLERTYPE_Masks
   sortPriority=1
   texture=&T_M_SM_Settlement_Flora_04_S
  }
  n28 = VectorParameter() {group=Emissive, parameterName=EmissiveADD, defaultValue=<rgba 0 0 0 0.1>}
  n29 = VertexColor()
  n30 = StaticSwitchParameter(True=NamedRerouteUsage(), False=@n29.A) {parameterName=UseUV2AnimMask}
  n31 = DotProduct(A=AppendVector(A=@n28.RGB, B=@n28.RGB), B=AppendVector(A=@n29, B=@n30))
  n32 = Multiply(
   A=LinearInterpolate(
    A=LinearInterpolate(A=@n24.RGB, B=@n25.RGB, Alpha=@n26)
    B=Multiply(A=@n19.Result, B=@n19.Result)
    Alpha=Fresnel() {baseReflectFraction=0, exponent=1.25}
   )
   B=Add(A=Multiply(A=@n14.R, B=@n27.G), B=Multiply(A=@n31, B=@n30))
  )
  n33 = CollectionParameter() {
   collection=&FortniteMaterialParameters
   parameterName=FortniteActiveTimeOfDay
  }
  n34 = VectorParameter() {group=Emissive, parameterName=TOD, channelNames {a=N, b=E, g=D, r=M}, defaultValue=<rgba 0.35 0.2 0.35 2>}
  n35 = DotProduct(A=@n33, B=AppendVector(A=@n34, B=@n34))
  n36 = TextureObject() {samplerType=SAMPLERTYPE_LinearGrayscale, texture=/Game/Global/Textures/Default/PointSplatter/T_GlowAlgae}
  n37 = Add(A=WorldPosition().XYZ, B=Multiply(A=VectorParameter() {parameterName=Scroll Spots, defaultValue=<rgba 0 0 -4 1>}.RGB, B=Time()))
  n38 = MaterialFunctionCall(
   Transition Phase=@n16
   Height Texture=MaterialFunctionCall(TextureObject=@n36, TextureSize=ScalarParameter() {defaultValue=32, parameterName=EmssiveSpotsScale}, WorldPosition=@n37) {materialFunction=&WorldAlignedTexture}.XYZ Texture
  ) {materialFunction=&HeightLerp}
  n39 = ScalarParameter() {defaultValue=0.25, group=Emissive, parameterName=EmissiveFlow Min}
  n40 = SetMaterialAttributes!(MaterialAttributes=@n23, Emissive Color=Multiply(A=Multiply(A=@n32, B=@n35), B=Max(A=@n38.Lerp Alpha No Contrast, B=@n39))) {attributeSetTypes [#7]}
  n41 = GetMaterialAttributes(None=@n40) {attributeGetTypes [#7, #1]}
  n42 = TextureSampleParameter2D() {
   group="Base "
   parameterName=Specular
   samplerSource=SSM_Wrap_WorldGroupSettings
   samplerType=SAMPLERTYPE_Masks
   sortPriority=1
   texture=&T_M_SM_Settlement_Flora_04_S
  }
  n43 = Multiply(
   A=MaterialFunctionCall(TextureObject=@n36, TextureSize=ScalarParameter() {defaultValue=75, parameterName=Em2 Spots Scale}, WorldPosition=Add(A=@n37, B=@n42.G)) {materialFunction=&WorldAlignedTexture}.XYZ Texture
   B=ScalarParameter() {defaultValue=2, group=Emissive 2, parameterName=Em2SpotsValue}
  )
  n44 = WorldPosition()
  n45 = Add(
   A=Multiply(A=VectorParameter() {parameterName=Em2 Scroll Speed, defaultValue=<rgba 0 0 -5 1>}.RGB, B=Time())
   B=Multiply(A=ScalarParameter() {defaultValue=55, group=Emissive 2, parameterName=Em2 Distortion}, B=@n42.G)
  )
  n46 = Max(
   A=MaterialFunctionCall(
    TextureObject=TextureObject() {texture=&T_ShieldCaustics}
    TextureSize=ScalarParameter() {defaultValue=512, group=Emissive 2, parameterName=Em2 TextureScale}
    WorldPosition=Add(A=@n44.XYZ, B=@n45)
   ) {materialFunction=&WorldAlignedTexture}.XYZ Texture
   B=ScalarParameter() {defaultValue=0.1, group=Emissive 2, parameterName=Em2 Central Glow Max}
  )
  n47 = NamedRerouteUsage()
  n48 = VertexColor()
  n49 = Multiply(
   A=Multiply(A=ScalarParameter() {defaultValue=5, group=Emissive 2, parameterName=Em 2 Value}, B=@n35)
   B=Multiply(A=Multiply(A=@n43, B=@n46), B=StaticSwitchParameter(True=@n47, False=@n48.A) {parameterName=UseUV2AnimMaskEm2})
  )
  n50 = DotProduct(A=CameraVectorWS(), B=VertexNormalWS())
  n51 = ScalarParameter() {defaultValue=1.5, group=Emissive 2, parameterName=Em2 Power Fresnel}
  n52 = Multiply(A=Add(A=@n41.Emissive Color, B=Multiply(A=@n41.Base Color, B=@n49)), B=Saturate(None=Power(Base=@n50, Exp=@n51)))
  n53 = SetMaterialAttributes!(
   MaterialAttributes=QualitySwitch(
    Default=@n40
    Low=@n40
    High=@n40
    Medium=@n40
    Epic=SetMaterialAttributes!(MaterialAttributes=@n41.MaterialAttributes, Emissive Color=@n52) {attributeSetTypes [#7]}
   )
   World Position Offset=Multiply(A=@n5, B=Add(A=@n6.Result, B=@n17))
  ) {attributeSetTypes [#8]}
  n54 = GetMaterialAttributes(None=@n53) {attributeGetTypes [#7, #1, #5]}
  n55 = SphereMask(
   A=CameraPositionWS()
   B=WorldPosition().XYZ
   Radius=LinearInterpolate(A=Constant() {r=750}, B=Constant() {r=200}, Alpha=PerInstanceRandom())
   Hardness=Constant() {r=0.65}
  )
  n56 = Constant3Vector() {constant=<rgba 0 0 1 0>}
  n57 = SphereMask(
   A=CameraPositionWS()
   B=WorldPosition().XYZ
   Radius=LinearInterpolate(A=Constant() {r=400}, B=Constant() {r=160}, Alpha=PerInstanceRandom())
   Hardness=Constant()
  )
  n58 = WorldPosition()
  n59 = CameraPositionWS()
  n60 = Add(A=Multiply(A=Subtract(A=@n58.XYZ, B=@n59), B=Constant3Vector() {constant=<rgba 1 1 0 0>}), B=Constant3Vector() {constant=<rgba 0 0 -200 0>})
  n61 = LinearInterpolate(A=ScalarParameter() {defaultValue=-75, group=Foliage, parameterName=SinkIntoGroundAmount}, B=Constant(), Alpha=PerInstanceFadeAmount())
  n62 = Constant3Vector() {constant=<rgba 0 0 1 0>}
  n63 = SetMaterialAttributes!(
   MaterialAttributes=@n54.MaterialAttributes
   World Position Offset=Add(A=Multiply(A=@n57, B=@n60), B=Multiply(A=@n61, B=@n62))
   Normal=LinearInterpolate(
    A=NamedRerouteUsage()
    B=Transform(None=@n56) {transformSourceType=TRANSFORMSOURCE_World, transformType=TRANSFORM_Tangent}
    Alpha=ScalarParameter() {defaultValue=0.5, group=Foliage, parameterName=Grass-Z-UpNormalsAmt}
   )
   Emissive Color=LinearInterpolate(A=@n54.Emissive Color, Alpha=@n55) {constB=0}
   Base Color=LinearInterpolate(A=@n54.Base Color, B=Multiply(A=@n54.Base Color) {constB=0.5}, Alpha=@n55) {constB=0}
   Subsurface Color=LinearInterpolate(A=@n54.Subsurface Color, B=Multiply(A=@n54.Subsurface Color) {constB=0.5}, Alpha=@n55) {constB=0}
  ) {attributeSetTypes [
    #8
    #6
    #7
    #1
    #5
   ]}
  n64 = GetMaterialAttributes(None=@n53) {attributeGetTypes [#8]}
  n65 = StaticBoolParameter() {defaultValue=true, group=z_code, parameterName=Code_UseHitEffects}
  n66 = ScalarParameter() {group=z_code, parameterName=Code_DestructionAmount}
  n67 = SetMaterialAttributes!(
   MaterialAttributes=@n64.MaterialAttributes
   World Position Offset=Add(
    A=@n64.World Position Offset
    B=MaterialFunctionCall(UseDestructibleDamage=@n65, DestructibleDamageAmount=@n66) {materialFunction=&MF_FN_DestructibleDamage}.DamageAnimWPO
   )
  ) {attributeSetTypes [#8]}
  MaterialFunctionCall(MA In=MaterialFunctionCall!(In=StaticSwitchParameter(True=@n63, False=@n67) {group=Foliage, parameterName=FoliageActor}, UseSphericalMappingOverride) {materialFunction=/Game/Packages/Fortress_SharedMaterials/Base_Material_Functions/MF_CurieFallbackEffect}.Result) {materialFunction=&MF_CreativeMoveTool}
  Reroute(None=@n7)
 ]
}
